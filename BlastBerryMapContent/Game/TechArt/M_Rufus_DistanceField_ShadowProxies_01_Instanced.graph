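# Unlit alpha-composite material for instanced shadow proxy cards that fake distance-field shadows, appearing as the camera rises.
# Vertex offset projects each card onto a horizontal plane along the clamped sun direction, then applies camera offset and water-level adjustment.
# Instance Y/Z scale offsets atlas UVs, and sun-height weights blend channels into a thresholded shadow shape.
# Time of day drives sharpness and opacity; scene-depth masking is bypassed on mobile.
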Material /BlastBerryMapContent/Game/TechArt/M_Rufus_DistanceField_ShadowProxies_01_Instanced
graph {
 properties {
  bApplyCloudFogging=true
  bComputeFogPerPixel=true
  bUsedWithInstancedStaticMeshes=true
  blendMode=BLEND_AlphaComposite
  opacityMaskClipValue=0.5
  pixelDepthOffsetMode=PDOM_Legacy
  shadingModel=MSM_Unlit
  usedShadingModels=Unlit
 }
 nodes [
  n1 = AppendVector(A=TextureCoordinate() {coordinateIndex=1}, B=ComponentMask(None=TextureCoordinate() {coordinateIndex=2}) {r=true}) {desc=Actor Pos}
  n2 = Multiply(A=TextureCoordinate()) {constB=4}
  n3 = TextureCoordinate() {coordinateIndex=1}
  n4 = ComponentMask(None=TextureCoordinate() {coordinateIndex=2}) {r=true}
  n5 = Sign(None=ComponentMask(None=AppendVector(A=@n3, B=@n4) {desc=Actor Pos}) {g=true})
  Add(
   A=MaterialFunctionCall!(
    Vector3=Multiply(A=@n1) {constB=3}
    MaximumNumberOfDigits
    UVs=Frac(None=@n2) {desc=0-1 UVs}
    DebugTextLocation RG_UpperRight BA_LowerLeft
    Component Spacing
   ) {materialFunction=/Engine/Functions/Engine_MaterialFunctions02/Utility/DebugFloat3Values}.ColorCodedOutput
   B=LinearInterpolate(A=Constant3Vector() {constant=<rgba 0.5 0 0.5 0>}, Alpha=Saturate(None=@n5)) {constB=0}
  )
  ComponentMask(None=TextureCoordinate() {coordinateIndex=2}) {g=true}
  ComponentMask(None=TextureCoordinate() {coordinateIndex=2}) {g=true}
  Constant3Vector() {constant=<rgba 1 0 0 0>}
  Constant()
  Constant() {r=1}
  n6 = CollectionParameter() {collection=&FortniteMaterialParameters, parameterName=FogColor}
  n7 = CollectionParameter() {
   collection=&FortniteMaterialParameters
   parameterName=SunlightDirectionalVector
  }
  n8 = Max(A=Abs(None=ComponentMask(None=@n7) {b=true}), B=ScalarParameter() {defaultValue=0.999, parameterName=MaxCos})
  n9 = Multiply(A=ComponentMask(None=@n8) {r=true}) {constB=3}
  n10 = Constant4Vector() {constant=<rgba 0 1 2 3>}
  n11 = OneMinus(None=Abs(None=Subtract(A=@n9, B=@n10.RGBA)))
  n12 = TextureCoordinate()
  n13 = TransformPosition(None=Constant3Vector()) {transformType=TRANSFORMPOSSOURCE_World}
  n14 = Constant3Vector() {constant=<rgba 0 1 0 0>}
  n15 = Constant3Vector() {constant=<rgba 0 0 1 0>}
  n16 = AppendVector(
   A=Distance(A=@n13, B=TransformPosition(None=@n14) {transformType=TRANSFORMPOSSOURCE_World})
   B=Distance(A=@n13, B=TransformPosition(None=@n15) {transformType=TRANSFORMPOSSOURCE_World})
  )
  n17 = TextureSampleParameter2D(UVs=Add(A=Divide(A=@n12) {constB=5, desc=Frames XY}, B=VertexInterpolator(VS=@n16).PS)) {parameterName=RT, samplerType=SAMPLERTYPE_LinearColor, texture=&/T__Rufus_ShadowProxy_22}
  n18 = DotProduct(A=VertexInterpolator(VS=Saturate(None=@n11)).PS, B=AppendVector(A=@n17.RGB, B=@n17.RGB))
  n19 = ScalarParameter() {defaultValue=-0.5, parameterName=EdgeShift}
  n20 = CollectionParameter() {
   collection=&FortniteMaterialParameters
   parameterName=FortniteActiveTimeOfDay
  }
  n21 = DotProduct(A=@n20, B=Constant4Vector() {constant=<rgba 1 0 0.5 1>}.RGBA)
  n22 = VertexInterpolator(VS=LinearInterpolate(
    A=ScalarParameter() {defaultValue=0.5, parameterName=DaySharpness}
    B=ScalarParameter() {defaultValue=0.7, parameterName=DuskDawnSharpness}
    Alpha=Saturate(None=@n21)
   ))
  n23 = Subtract(B=Add(A=Multiply(A=@n19, B=@n22.PS) {constA=-0.2}, B=@n22.PS)) {constA=0.5, constB=0.5}
  n24 = Saturate(None=Divide(A=Subtract(A=@n18, B=@n23) {constB=0}, B=@n22.PS))
  n25 = CameraPositionWS()
  n26 = Divide(
   A=Subtract(A=ComponentMask(None=@n25) {b=true}, B=ScalarParameter() {defaultValue=10000, parameterName=Camera Z Fadeout Loc})
   B=ScalarParameter() {defaultValue=35000, parameterName=Cam Z Fade Length}
  )
  n27 = ScalarParameter() {defaultValue=0.7, parameterName=Opacity}
  n28 = ScalarParameter() {defaultValue=0.5, parameterName=DuskDawnOpacity}
  n29 = CollectionParameter() {
   collection=&FortniteMaterialParameters
   parameterName=FortniteActiveTimeOfDay
  }
  n30 = Saturate(None=DotProduct(A=@n29, B=Constant4Vector() {constant=<rgba 1 0 0.5 1>}.RGBA))
  n31 = VertexInterpolator(VS=Multiply(A=Saturate(None=@n26) {desc=Fade Camera}, B=LinearInterpolate(A=@n27, B=@n28, Alpha=@n30)))
  n32 = Distance(
   A=TransformPosition(None=Constant3Vector()) {transformType=TRANSFORMPOSSOURCE_World}
   B=TransformPosition(None=Constant3Vector() {constant=<rgba 1 0 0 0>}) {transformType=TRANSFORMPOSSOURCE_World}
  )
  n33 = Multiply(
   A=MaterialFunctionCall!(
    Default=ScalarParameter() {defaultValue=2, parameterName=Offset}
    Mobile Override=ScalarParameter() {defaultValue=0.3, parameterName=Offset_Mobile}
   ) {materialFunction=/Game/ImpostorBaker/MaterialFunctions/OverrideMobile}.Result
   B=@n32
  )
  n34 = ScalarParameter() {parameterName=FadeOffset_Large}
  n35 = ScalarParameter() {parameterName=FadeOffset_Large_Mobile}
  n36 = Add(
   A=@n33
   B=Multiply(
    A=MaterialFunctionCall(Default=@n34, Mobile Override=@n35) {materialFunction=&OverrideMobile}.Result
    B=@n32
   )
  )
  n37 = Subtract(A=TransformPosition(None=Constant3Vector()) {transformType=TRANSFORMPOSSOURCE_World}, B=CameraPositionWS())
  n38 = Abs(None=DotProduct(
    A=Normalize(VectorInput=@n37)
    B=MaterialFunctionCall() {materialFunction=&CameraDirectionVector}.Result
   ))
  n39 = ScalarParameter() {defaultValue=1, parameterName=Shadow Fade Distance}
  n40 = ScalarParameter() {defaultValue=1, parameterName=Shadow Fade Distance_Mobile}
  n41 = SphereMask(
   A=Subtract(A=SceneDepth().R, B=PixelDepth().R)
   B=VertexInterpolator(VS=Multiply(A=@n36, B=@n38)).PS
   Radius=Multiply(
    A=MaterialFunctionCall(Default=@n39, Mobile Override=@n40) {materialFunction=&OverrideMobile}.Result
    B=VertexInterpolator(VS=@n32).PS
   )
   Hardness=ScalarParameter() {defaultValue=0.75, parameterName=ShadowHardness}
  ) {hardnessPercent=0}
  n42 = Constant() {r=1}
  LinearInterpolate(
   B=Multiply(A=ComponentMask(None=@n6) {b=true, g=true, r=true}, B=VectorParameter() {parameterName=ShadowColor, defaultValue=<rgba 0 0 0.01 0.01>}.A)
   Alpha=Multiply(
    A=Multiply(A=@n24, B=@n31.PS)
    B=MaterialFunctionCall(Default=@n41, Mobile Override=@n42) {materialFunction=&OverrideMobile}.Result
   )
  )
  MaterialFunctionCall() {materialFunction=&WorldPositionBehindTranslucency}
  n43 = CollectionParameter() {
   collection=&FortniteMaterialParameters
   parameterName=SunlightDirectionalVector
  }
  n44 = ComponentMask(None=@n43) {b=true}
  n45 = MaterialFunctionCall(
   Default=ScalarParameter() {defaultValue=0.5, parameterName=Max Cos PC}
   Mobile Override=ScalarParameter() {defaultValue=2, parameterName=Max Cos Mobile}
  ) {materialFunction=&OverrideMobile}
  n46 = Normalize(VectorInput=AppendVector(A=ComponentMask(None=@n43) {g=true, r=true}, B=Max(A=@n44, B=@n45.Result)))
  n47 = ComponentMask(None=@n46) {b=true, g=true, r=true}
  n48 = TextureCoordinate()
  n49 = MaterialFunctionCall(Float2=Multiply(A=ConstantBiasScale(None=@n48) {bias=-0.5, scale=2}, B=@n32) {constB=1000}) {materialFunction=&BreakOutFloat2Components}
  n50 = Normalize(VectorInput=CrossProduct(A=@n46, B=Constant3Vector() {constant=<rgba 0 0 1 0>}))
  n51 = Normalize(VectorInput=CrossProduct(A=@n46, B=@n50))
  n52 = Add(A=Add(A=Multiply(A=@n49.R, B=@n50), B=Multiply(A=@n49.G, B=@n51)), B=Multiply(A=@n32, B=Constant3Vector() {constant=<rgba 0 0 1 0>}))
  n53 = Multiply(A=Divide(A=@n47, B=ComponentMask(None=@n47) {b=true}), B=ComponentMask(None=@n52) {b=true})
  n54 = Add(A=Add(A=Multiply(A=@n53) {constB=-1}, B=@n52), B=TransformPosition(None=Constant3Vector()) {transformType=TRANSFORMPOSSOURCE_World})
  n55 = MaterialFunctionCall(Offset Amount=@n33, World Position=@n54) {materialFunction=/Engine/Functions/Engine_MaterialFunctions02/WorldPositionOffset/CameraOffset}
  n56 = WorldPosition() {worldPositionShaderOffset=WPT_ExcludeAllShaderOffsets}
  MaterialFunctionCall!(
   ImposterPosition=TransformPosition(None=Constant3Vector()) {transformType=TRANSFORMPOSSOURCE_World}
   WPO_Input=Add(A=Add(A=@n55.Camera Offset, B=@n54), B=Multiply(A=@n56.XYZ) {constB=-1})
   In
  ) {materialFunction=/Game/Athena/Apollo/Environments/Landscape/Materials/Biome/UniqueBiomeMaterialFunctions/MF_Apollo_WaterLevel}
  n57 = ArccosineFast!(None=@n8)
  n58 = Divide() {constA=57.2958, constB=90}
  Max(A=OneMinus(None=Multiply(A=@n57, B=@n58))) {constB=0.01}
  VectorParameter() {parameterName=asdfasf, defaultValue=<rgba 0.1 5 0.3 0>}
 ]
}
